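# Unlit translucent Niagara sprite material that draws glitchy digital noise blocks for the Sparks instrument swap effect.
# Two diagonally panning layers of a data-chip noise texture multiply together, with time quantized by Steps for choppy motion.
# Each layer samples at positive, zero and negative offsets for RGB chromatic split.
# Particle color tints the HDR-remapped emissive and desaturated alpha; dynamic parameters drive scale, speed and chroma, while BlockOpacity is unused.
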Material /SparksEffects/Effects/GameplayFX/InstrumentSwap/Materials/M_Sparks_Swap_DigitalNoiseBlocks
graph {
 properties {bUsedWithNiagaraSprites=true, blendMode=BLEND_Translucent, pixelDepthOffsetMode=PDOM_Legacy, shadingModel=MSM_Unlit, usedShadingModels=Unlit}
 nodes [
  n1 = ParticleColor()
  n2 = Multiply(A=TextureCoordinate(), B=DynamicParameter() {paramNames [UVScale,TimeSpeed,BlockOpacity,Chroma]}.UVScale)
  n3 = LinearInterpolate(
   A=Constant2Vector()
   B=ComponentMask(None=VectorParameter() {parameterName=ChromaOffset, defaultValue=<rgba 0.025 0.025 0 0>}.RGB) {g=true, r=true}
   Alpha=DynamicParameter() {paramNames [UVScale,TimeSpeed,BlockOpacity,Chroma]}.Chroma
  )
  n4 = Multiply(A=DynamicParameter() {paramNames [UVScale,TimeSpeed,BlockOpacity,Chroma]}.TimeSpeed, B=Time())
  n5 = ScalarParameter() {defaultValue=6, parameterName=Steps}
  n6 = Divide(A=Ceil(None=Multiply(A=@n4, B=@n5)), B=@n5)
  n7 = VectorParameter() {parameterName=BlocksPanSpeed, defaultValue=<rgba 1.8 0.1 -0.8 -0.62>}
  n8 = AppendVector(A=@n7.R, B=@n7.R)
  n9 = TextureSample(UVs=Panner(Coordinate=Add(A=@n2, B=@n3), Time=@n6, Speed=@n8) {bFractionalPart=true}) {texture=/Game/Athena/UI/Challenges/Art/T_UI_DataChip_Noise}
  n10 = TextureSample(UVs=Panner(Coordinate=@n2, Time=@n6, Speed=@n8) {bFractionalPart=true}) {texture=&T_UI_DataChip_Noise}
  n11 = Multiply(A=@n3) {constB=-1}
  n12 = TextureSample(UVs=Panner(Coordinate=Add(A=@n11, B=@n2), Time=@n6, Speed=@n8) {bFractionalPart=true}) {texture=&T_UI_DataChip_Noise}
  n13 = AppendVector(A=@n7.B, B=@n7.B)
  n14 = TextureSample(UVs=Panner(Coordinate=Add(A=@n3, B=@n2), Time=@n6, Speed=@n13) {bFractionalPart=true}) {texture=&T_UI_DataChip_Noise}
  n15 = TextureSample(UVs=Panner(Coordinate=@n2, Time=@n6, Speed=@n13) {bFractionalPart=true}) {texture=&T_UI_DataChip_Noise}
  n16 = TextureSample(UVs=Panner(Coordinate=Add(A=@n11, B=@n2), Time=@n6, Speed=@n13) {bFractionalPart=true}) {texture=&T_UI_DataChip_Noise}
  n17 = Multiply(
   A=@n1.RGB
   B=Multiply(
    A=MaterialFunctionCall(X=@n9.R, Y=@n10.R, Z=@n12.R) {materialFunction=&MakeFloat3}.Result
    B=MaterialFunctionCall(X=@n14.G, Y=@n15.G, Z=@n16.G) {materialFunction=&MakeFloat3}.Result
   )
  )
  n18 = Saturate(None=@n17)
  MaterialFunctionCall(Emissive=MaterialFunctionCall(LDR Input=@n18, HDR Tint=@n17) {materialFunction=&DeriveHDRfromLDR}.RemappedRange) {materialFunction=&MF_FX_EmissiveAdjustments}
  Multiply(A=@n1.A, B=Desaturation(None=@n18) {desc=Quick Alpha})
 ]
}
