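# Translucent, unlit, two-sided Niagara sprite material that builds a procedural square with an edge, filed under the InstrumentSwap gameplay effects.
# Two square masks come from mirrored UV distance to center, shaped by power and CheapContrast, with the second using center-scaled UVs.
# The second mask, weighted by the first dynamic parameter, is inverted to cut the interior from the first.
# Particle alpha multiplies the result, and no textures are sampled.
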Material /SparksEffects/Effects/GameplayFX/InstrumentSwap/Materials/M_Square_With_Edge
graph {
 properties {
  bUsedWithNiagaraSprites=true
  blendMode=BLEND_Translucent
  pixelDepthOffsetMode=PDOM_Legacy
  shadingModel=MSM_Unlit
  twoSided=true
  usedShadingModels=Unlit
 }
 nodes [
  n1 = MaterialFunctionCall(Float2=TextureCoordinate()) {materialFunction=&BreakOutFloat2Components}
  n2 = Abs(None=Subtract(A=@n1.R) {constB=0.5})
  n3 = ScalarParameter() {defaultValue=25, parameterName=sssssssssEdg}
  n4 = Multiply(A=OneMinus(None=Multiply(A=@n2) {constB=2}), B=@n3) {constB=35}
  n5 = Abs(None=Subtract(A=@n1.G) {constB=0.5})
  n6 = Multiply(A=OneMinus(None=Multiply(A=@n5) {constB=2}), B=@n3) {constB=5}
  n7 = MaterialFunctionCall(
   In=Power(Base=Multiply(A=@n4, B=@n6), Exp=ScalarParameter() {defaultValue=2, parameterName=F}) {constExponent=15}
   Contrast=ScalarParameter() {defaultValue=1, parameterName=C}
  ) {materialFunction=&CheapContrast}
  n8 = MaterialFunctionCall(Float2=MaterialFunctionCall(UVs=TextureCoordinate(), Texture Scale=ScalarParameter() {defaultValue=1.1, parameterName=TS}) {materialFunction=&ScaleUVsByCenter}.UVs) {materialFunction=&BreakOutFloat2Components}
  n9 = Abs(None=Subtract(A=@n8.R) {constB=0.5})
  n10 = ScalarParameter() {defaultValue=3, parameterName=sssssssssEdg2}
  n11 = Multiply(A=OneMinus(None=Multiply(A=@n9) {constB=2}), B=@n10) {constB=35}
  n12 = Abs(None=Subtract(A=@n8.G) {constB=0.5})
  n13 = Multiply(A=OneMinus(None=Multiply(A=@n12) {constB=2}), B=@n10) {constB=5}
  n14 = Power(Base=Multiply(A=Saturate(None=@n11), B=Saturate(None=@n13)), Exp=ScalarParameter() {defaultValue=1, parameterName=F2}) {constExponent=15}
  n15 = ScalarParameter() {defaultValue=0.2, parameterName=C2}
  n16 = Multiply(
   A=Saturate(None=MaterialFunctionCall(In=@n14, Contrast=@n15) {materialFunction=&CheapContrast}.Result)
   B=DynamicParameter() {defaultValue=<rgba 0.8 1 1 1>}.Param1
  )
  Multiply(A=ParticleColor().A, B=Multiply(A=Saturate(None=@n7.Result), B=OneMinus(None=@n16)))
 ]
}
