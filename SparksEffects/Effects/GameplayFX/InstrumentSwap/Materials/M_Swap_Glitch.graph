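# Unlit masked glitch-hologram material for instrument-swap effects, usable on Niagara mesh, sprite and ribbon particles.
# Emissive layers hue-shifted marbling noise, curve-atlas bands tinted by particle color and view angle, and moving screen-space lines.
# A mask chain combines scanlines and blocks with noise erosion and camera depth fade, driven partly by dynamic parameters.
# An offset chain offers wiggle, morph targets and rotation, but ends multiplied by zero.
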Material /SparksEffects/Effects/GameplayFX/InstrumentSwap/Materials/M_Swap_Glitch
graph {
 properties {
  bUsedWithNiagaraMeshParticles=true
  bUsedWithNiagaraRibbons=true
  bUsedWithNiagaraSprites=true
  blendMode=BLEND_Masked
  pixelDepthOffsetMode=PDOM_Legacy
  shadingModel=MSM_Unlit
  usedShadingModels=Unlit
 }
 nodes [
  n1 = DotProduct(A=PixelNormalWS(), B=CameraVectorWS())
  n2 = Add(A=AppendVector(A=@n1, B=@n1), B=Multiply(A=TextureCoordinate(), B=Constant() {r=2}))
  n3 = VectorParameter() {parameterName=HueShiftValue, defaultValue=<rgba 0.8 0 0 1>}
  n4 = DynamicParameter() {parameterIndex=1, defaultValue=<rgba 1 0 0 0>, paramNames [EdgeSpeed, BiasScale, Gradient Position, TriggerWiggle]}
  n5 = Add(A=Multiply(A=TextureCoordinate()) {constB=0.1}, B=@n4.EdgeSpeed)
  n6 = TextureSample(UVs=Panner(Coordinate=@n5) {speedX=0.05, speedY=0.1}) {texture=/Game/Characters/Player/Female/Medium/Bodies/F_MED_Veiled_Glitch/FX/T_Veiled_Glitch_Noise_Marbling}
  n7 = Panner(Coordinate=@n5) {speedX=-0.01, speedY=0.025}
  n8 = CurveAtlasRowParameter(CurveTime=Multiply(
    A=Multiply(A=@n6.G) {constB=2}
    B=TextureSample(UVs=@n7) {texture=&T_Veiled_Glitch_Noise_Marbling}.R
   )) {
   atlas=&AthenaCurveAtlas01
   curve=/Game/Athena/Environments/Nexus/Materials/FXshader/Curve_UnusedDelete
   defaultValue=303
   parameterName=CurveDetails
  }
  n9 = Saturate(None=Multiply(
    A=MaterialFunctionCall(Hue Shift Percentage=@n2, Texture=@n3.RGB) {materialFunction=&HueShift}.Result
    B=@n8
   ))
  n10 = ScalarParameter() {defaultValue=1.5, parameterName=ChromaticAmount}
  n11 = TextureCoordinate()
  n12 = DynamicParameter() {parameterIndex=1, defaultValue=<rgba 1 0 0 0>, paramNames [EdgeSpeed, BiasScale, Gradient Position, TriggerWiggle]}
  n13 = TextureSample(UVs=Multiply(A=Add(A=@n12.EdgeSpeed, B=@n11)) {constB=0.2}) {texture=&T_Veiled_Glitch_Noise_Marbling}
  n14 = ScalarParameter() {defaultValue=1, group=Dance Lines, parameterName=StretchGradient}
  n15 = SphereMask(
   A=MaterialFunctionCall() {materialFunction=&BreakOutFloat3Components}.G
   B=ScalarParameter() {defaultValue=150, desc=9.176, group=Dance Lines, parameterName=YAxisGradPosition}
   Radius=ScalarParameter() {defaultValue=50, desc=50, group=Dance Lines, parameterName=YAxisGradRadius}
   Hardness=Constant()
  )
  n16 = ScalarParameter() {defaultValue=1, group=Dance Lines, parameterName=YAxisGradIntensity}
  n17 = VectorParameter() {parameterName=ObjectSpaceGradStart, defaultValue=<rgba 0 40 150 1>}
  n18 = ScalarParameter() {defaultValue=180, group=Dance Lines, parameterName=ObjectSpaceGradRadius}
  n19 = Constant()
  n20 = Add(A=@n12.EdgeSpeed, B=PreSkinnedPosition())
  n21 = Add(
   A=@n12.EdgeSpeed
   B=Add(A=VectorParameter() {group=Dance Lines, parameterName=offset, defaultValue=<rgba 0 0 0 1>}.RGB, B=VertexInterpolator(VS=@n20).PS) {constB=0}
  )
  n22 = Add(
   A=Add(A=Multiply(A=@n15, B=@n16), B=SphereMask(A=@n21, B=@n17.RGB, Radius=@n18, Hardness=@n19))
   B=Multiply(B=ScalarParameter() {defaultValue=0.2, group=Dance Lines, parameterName=NormalDistortionWeight}) {constA=5}
  )
  n23 = Time()
  n24 = ScalarParameter() {defaultValue=1, group=Dance Lines, parameterName=speed}
  n25 = Add(
   A=Multiply(A=ComponentMask(None=@n13.RGB) {r=true}, B=ScalarParameter() {defaultValue=1, group=Dance Lines, parameterName=NoiseWarpIntensity})
   B=Add(A=Multiply(A=@n22, B=@n14), B=Multiply(A=@n23, B=@n24))
  )
  n26 = ScalarParameter() {defaultValue=0.75, group=Dance Lines, parameterName=Repetition}
  n27 = CurveAtlasRowParameter(CurveTime=Frac(None=Multiply(A=@n25, B=@n26))) {
   atlas=/Game/Characters/Player/Female/Medium/Bodies/F_MED_Veiled_Glitch/FX/CA_Veiled_Glitch
   curve=/Game/Characters/Player/Female/Medium/Bodies/F_MED_Veiled_Glitch/FX/CC_Veiled_Glitch
  }
  n28 = LinearInterpolate(
   A=@n27
   B=Desaturation(None=Normalize(VectorInput=@n27), Fraction=Constant() {r=5}) {luminanceFactors=<rgba 0 1 0.427 0>}
   Alpha=ScalarParameter() {parameterName=EmoteColor}
  )
  n29 = Clamp(None=@n28) {maxDefault=0.6, minDefault=0.4}
  n30 = ParticleColor()
  n31 = LinearInterpolate(
   A=Add(A=Multiply(A=@n29) {constB=0.1}, B=@n30.RGB)
   B=Multiply(A=@n30.RGB, B=Normalize(VectorInput=@n28))
   Alpha=ScalarParameter() {parameterName=GlitchNoise}
  )
  n32 = LinearInterpolate(
   A=@n31
   B=MaterialFunctionCall(Hue Shift Percentage=ScalarParameter() {defaultValue=0.7, parameterName=ShiftColor}, Texture=@n31) {materialFunction=&HueShift}.Result
   Alpha=ScalarParameter() {parameterName=UseShiftColor}
  )
  n33 = PixelNormalWS()
  n34 = CameraVectorWS()
  n35 = Power(Base=Clamp(None=DotProduct(A=@n33, B=@n34)), Exp=ScalarParameter() {defaultValue=2, parameterName=ViewDepedantFalloff})
  n36 = MaterialFunctionCall(World Position=WorldPosition().XYZ) {materialFunction=&TransformToClipSpace}
  n37 = Add(A=Time(), B=DynamicParameter() {defaultValue=<rgba 0 5 1 1>, paramNames [OpacityLerp, ScanLineRandom, Erosion, ErosionScale]}.ScanLineRandom)
  n38 = Subtract(A=@n36.Screen aligned 0-1 UV, B=ConstantBiasScale(None=Sine(None=@n37) {period=5.41}) {bias=1.3, scale=0.35})
  n39 = ScalarParameter() {defaultValue=96, parameterName=Line Multiplier A}
  n40 = StaticComponentMaskParameter(None=Clamp(None=Multiply(A=@n38, B=@n39))) {defaultG=true}
  n41 = Multiply(A=Multiply(A=@n40, B=OneMinus(None=@n40))) {constB=2.1}
  n42 = Subtract(A=@n36.Screen aligned 0-1 UV, B=ConstantBiasScale(None=Sine(None=@n37) {period=7.89}) {bias=1.25, scale=0.3})
  n43 = ScalarParameter() {defaultValue=84, parameterName=Line Multiplier B}
  n44 = StaticComponentMaskParameter(None=Clamp(None=Multiply(A=@n42, B=@n43))) {defaultG=true}
  n45 = Multiply(A=Multiply(A=@n44, B=OneMinus(None=@n44))) {constB=2.2}
  n46 = Multiply(A=Add(
    A=MaterialFunctionCall!(In=@n41) {materialFunction=/Engine/Functions/Engine_MaterialFunctions02/Utility/Round}.Result
    B=MaterialFunctionCall(In=@n45) {materialFunction=&Round}.Result
   )) {constB=0.05}
  n47 = Add(A=Power(Base=Multiply(A=@n9, B=@n10) {constB=1.5}), B=Add(A=Multiply(A=@n35, B=@n32), B=@n46))
  MaterialFunctionCall(Emissive=@n47) {materialFunction=&MF_FX_EmissiveAdjustments}
  n48 = ScalarParameter() {defaultValue=5, parameterName=FadeLength}
  n49 = ScalarParameter() {defaultValue=80, parameterName=FadeOffset}
  n50 = DynamicParameter() {parameterIndex=1, defaultValue=<rgba 1 0 0 0>, paramNames [EdgeSpeed, BiasScale, Gradient Position, TriggerWiggle]}
  n51 = MaterialFunctionCall!(
   WPO Intensity A
   WPO Intensity B
   Time
   Bias=StaticSwitchParameter(True=@n50.TriggerWiggle, False=Constant()) {defaultValue=true, parameterName=Use Param4 for Wiggle}
  ) {materialFunction=/Game/Effects/Fort_Effects/Effects/Atlas/Materials/MF_Atlas_HologramWiggle_01}
  n52 = Constant()
  n53 = MaterialFunctionCall!(
   Number of Scanlines=ScalarParameter() {defaultValue=150, parameterName=Scanline Amount}
   Scanline Speed
   "Scanlines Thickness (Bias)"=ScalarParameter() {defaultValue=-0.15, parameterName=ScanlineSize}
   Scanlines Dark Value=ScalarParameter() {defaultValue=0.5, parameterName=Scanlines Dark}
   Scanlines Light Value=ScalarParameter() {defaultValue=0.6, parameterName=Scanlines Light}
   Block Opacity=Multiply(A=StaticSwitchParameter(True=@n51.Intense 0-1, False=@n52) {defaultValue=true, parameterName=Use Blocks?}) {constB=0.4}
   UVs
   Time
   Block Repeat
  ) {materialFunction=/Game/Effects/Fort_Effects/Effects/Atlas/Materials/MF_Atlas_Scanlines_01}
  n54 = TextureCoordinate()
  n55 = DynamicParameter() {defaultValue=<rgba 0 5 1 1>, paramNames [OpacityLerp, ScanLineRandom, Erosion, ErosionScale]}
  n56 = LinearInterpolate(
   A=ScalarParameter() {defaultValue=1, parameterName=ErosionScale}
   B=@n55.ErosionScale
   Alpha=ScalarParameter() {defaultValue=1, parameterName=UseErosionScaleDynParam}
  )
  n57 = Subtract(A=TextureSample(UVs=Multiply(A=@n54, B=@n56)) {texture=/Game/Athena/UI/Challenges/Art/T_UI_DataChip_Noise}.G) {constB=0.5}
  Multiply(
   A=Multiply(A=Add(A=@n46, B=@n53.Result), B=LinearInterpolate(A=@n57, Alpha=@n55.Erosion) {constAlpha=0})
   B=Saturate(None=MaterialFunctionCall(Fade Length=@n48, Fade Offset=@n49) {materialFunction=&CameraDepthFade}.Result)
  )
  n58 = Constant3Vector() {constant=<rgba 0 0 1 0>}
  n59 = RotateAboutAxis(
   NormalizedRotationAxis=Normalize(VectorInput=Transform(None=@n58) {transformSourceType=TRANSFORMSOURCE_Local})
   RotationAngle=Multiply(A=Time(), B=ScalarParameter() {defaultValue=0.05, parameterName=RotationSpeed})
   PivotPoint=ObjectPositionWS()
   Position=WorldPosition().XYZ
  )
  n60 = Multiply(A=ScalarParameter() {defaultValue=0.1, parameterName=Time Speed}, B=Time() {period=1})
  n61 = TextureObjectParameter() {
   parameterName=Morph Target Normals
   samplerType=SAMPLERTYPE_LinearColor
   texture=/Game/Global/Textures/Default/Blanks/T_BlackVectorDisplacement
  }
  n62 = TextureObjectParameter() {
   parameterName=Morph Target Offsets
   samplerType=SAMPLERTYPE_LinearColor
   texture=&T_BlackVectorDisplacement
  }
  n63 = ComponentMask(None=TextureProperty(Tex=@n61)) {g=true}
  n64 = StaticBool() {value=true}
  n65 = StaticSwitchParameter(
   True=Add(
    A=@n51.Result
    B=MaterialFunctionCall!(Morph Animation=@n60, Morph Normal=@n61, Morph Texture=@n62, Number of Morph Targets=@n63, 0-1 Animation Value?=@n64, Morph Target UV) {materialFunction=/Engine/Functions/Engine_MaterialFunctions02/MAXScripts/MS_VertexAnimationTools_MorphTargets}.World Position Offset
   )
   False=@n51.Result
  ) {parameterName=Use Morph Targets?}
  n66 = FeatureLevelSwitch(Default=@n59, ES2=@n65, ES3_1=@n65, SM4_REMOVED=@n59)
  Multiply(A=StaticSwitchParameter(True=StaticSwitchParameter(True=@n66, False=@n65) {parameterName=Use Rotation}, False=Constant4Vector().RGBA) {defaultValue=true, parameterName=Use WPO}) {constB=0}
  Reroute(None=@n47)
 ]
}
